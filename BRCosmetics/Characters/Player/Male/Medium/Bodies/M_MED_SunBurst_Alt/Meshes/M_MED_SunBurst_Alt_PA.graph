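# Physics asset for the SunBurst Alt medium male player body cosmetic, set up for physics on its coat, collar and lapels.
# Root, pelvis, spine, neck, head and leg bodies use free-angular constraints and anchor small "mstr" spheres.
# From these spheres hang limited-swing "dyn" capsules: six four-segment coat chains at the pelvis, plus collar and lapel pieces on the upper spine.
# Free-angular constraints also link neighboring collar capsules from front-left around the back to front-right.
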PhysicsAsset /BRCosmetics/Characters/Player/Male/Medium/Bodies/M_MED_SunBurst_Alt/Meshes/M_MED_SunBurst_Alt_PA
bodies {
 calf_l [{shapeType=Capsule, length=25, radius=8, shapeName=calf_l_capsule, center=<-20 0 0>, rotation=<rot 90 0 0>}]
 calf_r [{shapeType=Capsule, length=35, radius=15, shapeName=calf_r_capsule, center=<25 0 0>, rotation=<rot 90 0 0>}]
 coat_bl_01_dyn [{shapeType=Capsule, length=10, radius=9, shapeName=coat_bl_01_dyn_capsule, center=<2 0 6>, rotation=<rot 0 0 90>}]
 coat_bl_01_mstr [{shapeType=Sphere, radius=0.505, shapeName=coat_bl_01_mstr_sphere}]
 coat_bl_02_dyn [{shapeType=Capsule, length=10, radius=9, shapeName=coat_bl_02_dyn_capsule, center=<6 0 6>, rotation=<rot 0 0 90>}]
 coat_bl_03_dyn [{shapeType=Capsule, length=10, radius=9, shapeName=coat_bl_03_dyn_capsule, center=<8 0 6>, rotation=<rot 0 0 90>}]
 coat_bl_04_dyn [{shapeType=Capsule, length=10, radius=9, shapeName=coat_bl_04_dyn_capsule, center=<10 0 8>, rotation=<rot 0 0 90>}]
 coat_br_01_dyn [{shapeType=Capsule, length=10, radius=9, shapeName=coat_br_01_dyn_capsule, center=<2 0 6>, rotation=<rot 0 0 90>}]
 coat_br_01_mstr [{shapeType=Sphere, radius=0.505, shapeName=coat_br_01_mstr_sphere}]
 coat_br_02_dyn [{shapeType=Capsule, length=10, radius=9, shapeName=coat_br_02_dyn_capsule, center=<6 0 6>, rotation=<rot 0 0 90>}]
 coat_br_03_dyn [{shapeType=Capsule, length=25, radius=9, shapeName=coat_br_03_dyn_capsule, center=<8 0 6>, rotation=<rot 0 0 90>}]
 coat_br_04_dyn [{shapeType=Capsule, length=25, radius=9, shapeName=coat_br_04_dyn_capsule, center=<10 0 6>, rotation=<rot 0 0 90>}]
 coat_fl_01_dyn [{shapeType=Capsule, length=5, radius=9, shapeName=coat_fl_01_dyn_capsule, center=<2 -8 6>, rotation=<rot 0 0 60>}]
 coat_fl_01_mstr [{shapeType=Sphere, radius=0.505, shapeName=coat_fl_01_mstr_sphere}]
 coat_fl_02_dyn [{shapeType=Capsule, length=10, radius=9, shapeName=coat_fl_02_dyn_capsule, center=<6 -8 -6>}]
 coat_fl_03_dyn [{shapeType=Capsule, length=10, radius=9, shapeName=coat_fl_03_dyn_capsule, center=<10 -7 -6>}]
 coat_fl_04_dyn [{shapeType=Capsule, length=10, radius=9, shapeName=coat_fl_04_dyn_capsule, center=<8 -7 -6>}]
 coat_fr_01_dyn [{shapeType=Capsule, length=5, radius=6, shapeName=coat_fr_01_dyn_capsule, center=<2 6 6>, rotation=<rot 0 0 -60>}]
 coat_fr_01_mstr [{shapeType=Sphere, radius=0.505, shapeName=coat_fr_01_mstr_sphere}]
 coat_fr_02_dyn [{shapeType=Capsule, length=10, radius=10, shapeName=coat_fr_02_dyn_capsule, center=<6 6 6>, rotation=<rot 0 0 -60>}]
 coat_fr_03_dyn [{shapeType=Capsule, length=10, radius=10, shapeName=coat_fr_03_dyn_capsule, center=<10 6 6>, rotation=<rot 0 0 -60>}]
 coat_fr_04_dyn [{shapeType=Capsule, length=10, radius=10, shapeName=coat_fr_04_dyn_capsule, center=<8 6 6>, rotation=<rot 0 0 -60>}]
 coat_l_01_dyn [{shapeType=Capsule, length=10, radius=5, shapeName=coat_l_01_dyn_capsule, center=<2 0 6>, rotation=<rot 0 0 90>}]
 coat_l_01_mstr [{shapeType=Sphere, radius=0.505, shapeName=coat_l_01_mstr_sphere}]
 coat_l_02_dyn [{shapeType=Capsule, length=10, radius=7, shapeName=coat_l_02_dyn_capsule, center=<6 0 4>, rotation=<rot 0 0 90>}]
 coat_l_03_dyn [{shapeType=Capsule, length=10, radius=7, shapeName=coat_l_03_dyn_capsule, center=<10 0 4>, rotation=<rot 0 0 90>}]
 coat_l_04_dyn [{shapeType=Capsule, length=10, radius=7, shapeName=coat_l_04_dyn_capsule, center=<8 0 4>, rotation=<rot 0 0 90>}]
 coat_r_01_dyn [{shapeType=Capsule, length=10, radius=5, shapeName=coat_r_01_dyn_capsule, center=<2 0 6>, rotation=<rot 0 0 90>}]
 coat_r_01_mstr [{shapeType=Sphere, radius=0.505, shapeName=coat_r_01_mstr_sphere}]
 coat_r_02_dyn [{shapeType=Capsule, length=10, radius=9, shapeName=coat_r_02_dyn_capsule, center=<6 0 8>, rotation=<rot 0 0 90>}]
 coat_r_03_dyn [{shapeType=Capsule, length=10, radius=9, shapeName=coat_r_03_dyn_capsule, center=<10 0 8>, rotation=<rot 0 0 90>}]
 coat_r_04_dyn [{shapeType=Capsule, length=10, radius=9, shapeName=coat_r_04_dyn_capsule, center=<8 0 8>, rotation=<rot 0 0 90>}]
 collar_b_dyn [{shapeType=Capsule, length=5, radius=5.5, shapeName=collar_b_dyn_capsule, center=<6 0 0>, rotation=<rot 0 0 90>}]
 collar_b_mstr [{shapeType=Sphere, radius=1, shapeName=collar_b_mstr_sphere}]
 collar_bl_dyn [{shapeType=Capsule, length=5, radius=4, shapeName=collar_r_dyn_capsule, center=<7 0 1>, rotation=<rot 0 0 90>}]
 collar_bl_mstr [{shapeType=Sphere, radius=0.5, shapeName=collar_bl_mstr_sphere}]
 collar_br_dyn [{shapeType=Capsule, length=5, radius=4, shapeName=collar_r_dyn_capsule, center=<7 0 1>, rotation=<rot 0 0 90>}]
 collar_br_mstr [{shapeType=Sphere, radius=0.5, shapeName=collar_br_mstr_sphere}]
 collar_fl_dyn [{shapeType=Capsule, length=5, radius=4, shapeName=collar_fl_dyn_capsule, center=<6 0 0>, rotation=<rot 0 0 90>}]
 collar_fl_mstr [{shapeType=Sphere, radius=1, shapeName=collar_fl_mstr_sphere}]
 collar_fr_dyn [{shapeType=Capsule, length=5, radius=4, shapeName=collar_fr_dyn_capsule, center=<6 0 0>, rotation=<rot 0 0 90>}]
 collar_fr_mstr [{shapeType=Sphere, radius=1, shapeName=collar_fr_mstr_sphere}]
 collar_l_dyn [{shapeType=Capsule, length=5, radius=4, shapeName=collar_l_dyn_capsule, center=<6 0 0>, rotation=<rot 0 0 90>}]
 collar_l_mstr [{shapeType=Sphere, radius=1, shapeName=collar_l_mstr_sphere}]
 collar_r_dyn [{shapeType=Capsule, length=5, radius=4, shapeName=collar_r_dyn_capsule, center=<6 0 0>, rotation=<rot 0 0 90>}]
 collar_r_mstr [{shapeType=Sphere, radius=1, shapeName=collar_r_mstr_sphere}]
 foot_r [{shapeType=Capsule, length=12, radius=12, shapeName=foot_r_capsule, center=<2 -6 0>, rotation=<rot 0 0 90>}]
 head [{shapeType=Capsule, length=6, radius=8, shapeName=head_capsule, center=<2 0 0>}]
 lapel_l_dyn [{shapeType=Capsule, length=5, radius=3, shapeName=lapel_l_dyn_capsule, center=<5 0 0>, rotation=<rot 90 0 0>}]
 lapel_l_mstr [{shapeType=Sphere, radius=1, shapeName=lapel_l_mstr_sphere}]
 lapel_low_l_dyn [{shapeType=Capsule, length=5, radius=3, shapeName=lapel_low_l_dyn_capsule, center=<5 0 0>, rotation=<rot 90 0 0>}]
 lapel_low_l_mstr [{shapeType=Sphere, radius=1, shapeName=lapel_low_l_mstr_sphere}]
 lapel_low_r_dyn [{shapeType=Capsule, length=5, radius=3, shapeName=lapel_low_r_dyn_capsule, center=<5 0 0>, rotation=<rot 90 0 0>}]
 lapel_low_r_mstr [{shapeType=Sphere, radius=1, shapeName=lapel_low_r_mstr_sphere}]
 lapel_r_dyn [{shapeType=Capsule, length=5, radius=3, shapeName=lapel_r_dyn_capsule, center=<5 0 0>, rotation=<rot 90 0 0>}]
 lapel_r_mstr [{shapeType=Sphere, radius=1, shapeName=lapel_r_mstr_sphere}]
 neck_02 [{shapeType=Capsule, length=8, radius=5, shapeName=neck_02_capsule, rotation=<rot 90 0 0>}]
 pelvis [{shapeType=Capsule, length=15, radius=12}]
 root [{shapeType=Sphere, radius=25, shapeName=root_sphere}]
 spine_01 [{shapeType=Capsule, length=12, radius=12, shapeName=spine_01_capsule}]
 spine_02 [{shapeType=Capsule, length=12, radius=12, shapeName=spine_02_capsule}]
 spine_04 [{shapeType=Capsule, length=12, radius=12, shapeName=spine_04_capsule}]
 spine_05 [{shapeType=Capsule, length=20, radius=10, shapeName=spine_05_capsule}]
 thigh_l [{shapeType=Capsule, length=25, radius=10, shapeName=thigh_l_capsule, center=<-16 0 0>, rotation=<rot 90 0 0>}]
 thigh_r [{shapeType=Capsule, length=25, radius=10, shapeName=thigh_r_capsule, center=<16 0 0>, rotation=<rot 90 0 0>}]
 thigh_twist_01_l [{shapeType=Capsule, length=25, radius=8, shapeName=thigh_twist_01_l_capsule, rotation=<rot 90 88.3838 88.3838>}]
 thigh_twist_01_r [{shapeType=Capsule, length=25, radius=8, shapeName=thigh_twist_01_r_capsule, rotation=<rot 90 0 0>}]
}
constraints [
 {
  bone1Name=spine_01
  bone2Name=pelvis
  swing1LimitDegrees=45
  swing1Motion=Free
  swing2LimitDegrees=45
  swing2Motion=Free
  twistLimitDegrees=45
  twistMotion=Free
 }
 {
  bone1Name=spine_02
  bone2Name=spine_01
  swing1LimitDegrees=45
  swing1Motion=Free
  swing2LimitDegrees=45
  swing2Motion=Free
  twistLimitDegrees=45
  twistMotion=Free
 }
 {
  bone1Name=spine_04
  bone2Name=spine_02
  swing1LimitDegrees=45
  swing1Motion=Free
  swing2LimitDegrees=45
  swing2Motion=Free
  twistLimitDegrees=45
  twistMotion=Free
 }
 {
  bone1Name=spine_05
  bone2Name=spine_04
  swing1LimitDegrees=45
  swing1Motion=Free
  swing2LimitDegrees=45
  swing2Motion=Free
  twistLimitDegrees=45
  twistMotion=Free
 }
 {
  bone1Name=lapel_low_l_mstr
  bone2Name=spine_04
  swing1LimitDegrees=45
  swing1Motion=Free
  swing2LimitDegrees=45
  swing2Motion=Free
  twistLimitDegrees=45
  twistMotion=Free
 }
 {bone1Name=lapel_low_l_dyn, bone2Name=lapel_low_l_mstr, swing1LimitDegrees=5, swing2LimitDegrees=15, twistLimitDegrees=10}
 {
  bone1Name=lapel_low_r_mstr
  bone2Name=spine_04
  swing1LimitDegrees=45
  swing1Motion=Free
  swing2LimitDegrees=45
  swing2Motion=Free
  twistLimitDegrees=45
  twistMotion=Free
 }
 {bone1Name=lapel_low_r_dyn, bone2Name=lapel_low_r_mstr, swing1LimitDegrees=5, swing2LimitDegrees=15, twistLimitDegrees=10}
 {
  bone1Name=collar_b_mstr
  bone2Name=spine_05
  swing1LimitDegrees=45
  swing1Motion=Free
  swing2LimitDegrees=45
  swing2Motion=Free
  twistLimitDegrees=45
  twistMotion=Free
 }
 {
  bone1Name=collar_fl_mstr
  bone2Name=spine_05
  swing1LimitDegrees=45
  swing1Motion=Free
  swing2LimitDegrees=45
  swing2Motion=Free
  twistLimitDegrees=45
  twistMotion=Free
 }
 {
  bone1Name=collar_fr_mstr
  bone2Name=spine_05
  swing1LimitDegrees=45
  swing1Motion=Free
  swing2LimitDegrees=45
  swing2Motion=Free
  twistLimitDegrees=45
  twistMotion=Free
 }
 {
  bone1Name=collar_l_mstr
  bone2Name=spine_05
  swing1LimitDegrees=45
  swing1Motion=Free
  swing2LimitDegrees=45
  swing2Motion=Free
  twistLimitDegrees=45
  twistMotion=Free
 }
 {
  bone1Name=collar_r_mstr
  bone2Name=spine_05
  swing1LimitDegrees=45
  swing1Motion=Free
  swing2LimitDegrees=45
  swing2Motion=Free
  twistLimitDegrees=45
  twistMotion=Free
 }
 {
  bone1Name=lapel_l_mstr
  bone2Name=spine_05
  swing1LimitDegrees=45
  swing1Motion=Free
  swing2LimitDegrees=45
  swing2Motion=Free
  twistLimitDegrees=45
  twistMotion=Free
 }
 {bone1Name=lapel_l_dyn, bone2Name=lapel_l_mstr, swing1LimitDegrees=5, swing2LimitDegrees=15, twistLimitDegrees=10}
 {
  bone1Name=lapel_r_mstr
  bone2Name=spine_05
  swing1LimitDegrees=45
  swing1Motion=Free
  swing2LimitDegrees=45
  swing2Motion=Free
  twistLimitDegrees=45
  twistMotion=Free
 }
 {bone1Name=lapel_r_dyn, bone2Name=lapel_r_mstr, swing1LimitDegrees=5, swing2LimitDegrees=15, twistLimitDegrees=10}
 {
  bone1Name=thigh_l
  bone2Name=pelvis
  swing1LimitDegrees=45
  swing1Motion=Free
  swing2LimitDegrees=45
  swing2Motion=Free
  twistLimitDegrees=45
  twistMotion=Free
 }
 {
  bone1Name=calf_l
  bone2Name=thigh_l
  swing1LimitDegrees=45
  swing1Motion=Free
  swing2LimitDegrees=45
  swing2Motion=Free
  twistLimitDegrees=45
  twistMotion=Free
 }
 {
  bone1Name=thigh_r
  bone2Name=pelvis
  swing1LimitDegrees=45
  swing1Motion=Free
  swing2LimitDegrees=45
  swing2Motion=Free
  twistLimitDegrees=45
  twistMotion=Free
 }
 {
  bone1Name=calf_r
  bone2Name=thigh_r
  swing1LimitDegrees=45
  swing1Motion=Free
  swing2LimitDegrees=45
  swing2Motion=Free
  twistLimitDegrees=45
  twistMotion=Free
 }
 {
  bone1Name=neck_02
  bone2Name=spine_05
  swing1LimitDegrees=45
  swing1Motion=Free
  swing2LimitDegrees=45
  swing2Motion=Free
  twistLimitDegrees=45
  twistMotion=Free
 }
 {
  bone1Name=head
  bone2Name=neck_02
  swing1LimitDegrees=45
  swing1Motion=Free
  swing2LimitDegrees=45
  swing2Motion=Free
  twistLimitDegrees=45
  twistMotion=Free
 }
 {
  bone1Name=thigh_twist_01_l
  bone2Name=thigh_l
  swing1LimitDegrees=45
  swing1Motion=Free
  swing2LimitDegrees=45
  swing2Motion=Free
  twistLimitDegrees=45
  twistMotion=Free
 }
 {
  bone1Name=thigh_twist_01_r
  bone2Name=thigh_r
  swing1LimitDegrees=45
  swing1Motion=Free
  swing2LimitDegrees=45
  swing2Motion=Free
  twistLimitDegrees=45
  twistMotion=Free
 }
 {bone1Name=collar_b_dyn, bone2Name=collar_b_mstr, swing1LimitDegrees=25, swing2LimitDegrees=5, twistLimitDegrees=5}
 {bone1Name=collar_fl_dyn, bone2Name=collar_fl_mstr, swing1LimitDegrees=25, swing2LimitDegrees=5, twistLimitDegrees=5}
 {bone1Name=collar_fr_dyn, bone2Name=collar_fr_mstr, swing1LimitDegrees=25, swing2LimitDegrees=5, twistLimitDegrees=5}
 {bone1Name=collar_l_dyn, bone2Name=collar_l_mstr, swing1LimitDegrees=25, swing2LimitDegrees=5, twistLimitDegrees=5}
 {bone1Name=collar_r_dyn, bone2Name=collar_r_mstr, swing1LimitDegrees=25, swing2LimitDegrees=5, twistLimitDegrees=5}
 {bone1Name=coat_fr_01_dyn, bone2Name=coat_fr_01_mstr, swing1LimitDegrees=15, swing2LimitDegrees=35, twistLimitDegrees=15}
 {bone1Name=coat_r_01_dyn, bone2Name=coat_r_01_mstr, swing1LimitDegrees=35, swing2LimitDegrees=35, twistLimitDegrees=15}
 {bone1Name=coat_bl_01_dyn, bone2Name=coat_bl_01_mstr, swing1LimitDegrees=35, swing2LimitDegrees=35, twistLimitDegrees=15}
 {bone1Name=coat_br_01_dyn, bone2Name=coat_br_01_mstr, swing1LimitDegrees=35, swing2LimitDegrees=35, twistLimitDegrees=15}
 {
  bone1Name=coat_br_01_mstr
  bone2Name=pelvis
  swing1LimitDegrees=45
  swing1Motion=Free
  swing2LimitDegrees=45
  swing2Motion=Free
  twistLimitDegrees=45
  twistMotion=Free
 }
 {
  bone1Name=coat_fl_01_mstr
  bone2Name=pelvis
  swing1LimitDegrees=45
  swing1Motion=Free
  swing2LimitDegrees=45
  swing2Motion=Free
  twistLimitDegrees=45
  twistMotion=Free
 }
 {
  bone1Name=coat_fr_01_mstr
  bone2Name=pelvis
  swing1LimitDegrees=45
  swing1Motion=Free
  swing2LimitDegrees=45
  swing2Motion=Free
  twistLimitDegrees=45
  twistMotion=Free
 }
 {
  bone1Name=coat_r_01_mstr
  bone2Name=pelvis
  swing1LimitDegrees=45
  swing1Motion=Free
  swing2LimitDegrees=45
  swing2Motion=Free
  twistLimitDegrees=45
  twistMotion=Free
 }
 {
  bone1Name=coat_bl_01_mstr
  bone2Name=pelvis
  swing1LimitDegrees=45
  swing1Motion=Free
  swing2LimitDegrees=45
  swing2Motion=Free
  twistLimitDegrees=45
  twistMotion=Free
 }
 {bone1Name=coat_l_01_dyn, bone2Name=coat_l_01_mstr, swing1LimitDegrees=35, swing2LimitDegrees=35, twistLimitDegrees=15}
 {
  bone1Name=coat_l_01_mstr
  bone2Name=pelvis
  swing1LimitDegrees=45
  swing1Motion=Free
  swing2LimitDegrees=45
  swing2Motion=Free
  twistLimitDegrees=45
  twistMotion=Free
 }
 {bone1Name=coat_fl_01_dyn, bone2Name=coat_fl_01_mstr, swing1LimitDegrees=35, swing2LimitDegrees=35, twistLimitDegrees=15}
 {
  bone1Name=collar_bl_mstr
  bone2Name=spine_05
  swing1LimitDegrees=45
  swing1Motion=Free
  swing2LimitDegrees=45
  swing2Motion=Free
  twistLimitDegrees=45
  twistMotion=Free
 }
 {bone1Name=collar_br_dyn, bone2Name=collar_br_mstr, swing1LimitDegrees=25, swing2LimitDegrees=5, twistLimitDegrees=5}
 {
  bone1Name=collar_br_mstr
  bone2Name=spine_05
  swing1LimitDegrees=45
  swing1Motion=Free
  swing2LimitDegrees=45
  swing2Motion=Free
  twistLimitDegrees=45
  twistMotion=Free
 }
 {bone1Name=collar_bl_dyn, bone2Name=collar_bl_mstr, swing1LimitDegrees=25, swing2LimitDegrees=5, twistLimitDegrees=5}
 {bone1Name=coat_br_02_dyn, bone2Name=coat_br_01_dyn, swing1LimitDegrees=35, swing2LimitDegrees=35, twistLimitDegrees=15}
 {bone1Name=coat_br_03_dyn, bone2Name=coat_br_02_dyn, swing1LimitDegrees=35, swing2LimitDegrees=35, twistLimitDegrees=15}
 {bone1Name=coat_br_04_dyn, bone2Name=coat_br_03_dyn, swing1LimitDegrees=35, swing2LimitDegrees=35, twistLimitDegrees=15}
 {bone1Name=coat_fl_02_dyn, bone2Name=coat_fl_01_dyn, swing1LimitDegrees=35, swing2LimitDegrees=35, twistLimitDegrees=15}
 {bone1Name=coat_fl_03_dyn, bone2Name=coat_fl_02_dyn, swing1LimitDegrees=35, swing2LimitDegrees=35, twistLimitDegrees=15}
 {bone1Name=coat_fl_04_dyn, bone2Name=coat_fl_03_dyn, swing1LimitDegrees=35, swing2LimitDegrees=35, twistLimitDegrees=15}
 {bone1Name=coat_fr_02_dyn, bone2Name=coat_fr_01_dyn, swing1LimitDegrees=15, swing2LimitDegrees=35, twistLimitDegrees=15}
 {bone1Name=coat_fr_03_dyn, bone2Name=coat_fr_02_dyn, swing1LimitDegrees=15, swing2LimitDegrees=35, twistLimitDegrees=15}
 {bone1Name=coat_fr_04_dyn, bone2Name=coat_fr_03_dyn, swing1LimitDegrees=15, swing2LimitDegrees=35, twistLimitDegrees=15}
 {bone1Name=coat_l_02_dyn, bone2Name=coat_l_01_dyn, swing1LimitDegrees=35, swing2LimitDegrees=35, twistLimitDegrees=15}
 {bone1Name=coat_l_03_dyn, bone2Name=coat_l_02_dyn, swing1LimitDegrees=35, swing2LimitDegrees=35, twistLimitDegrees=15}
 {bone1Name=coat_l_04_dyn, bone2Name=coat_l_03_dyn, swing1LimitDegrees=35, swing2LimitDegrees=35, twistLimitDegrees=15}
 {bone1Name=coat_r_02_dyn, bone2Name=coat_r_01_dyn, swing1LimitDegrees=35, swing2LimitDegrees=35, twistLimitDegrees=15}
 {bone1Name=coat_r_03_dyn, bone2Name=coat_r_02_dyn, swing1LimitDegrees=35, swing2LimitDegrees=35, twistLimitDegrees=15}
 {bone1Name=coat_r_04_dyn, bone2Name=coat_r_03_dyn, swing1LimitDegrees=35, swing2LimitDegrees=35, twistLimitDegrees=15}
 {bone1Name=coat_bl_04_dyn, bone2Name=coat_bl_03_dyn, swing1LimitDegrees=35, swing2LimitDegrees=35, twistLimitDegrees=15}
 {bone1Name=coat_bl_03_dyn, bone2Name=coat_bl_02_dyn, swing1LimitDegrees=35, swing2LimitDegrees=35, twistLimitDegrees=15}
 {bone1Name=coat_bl_02_dyn, bone2Name=coat_bl_01_dyn, swing1LimitDegrees=35, swing2LimitDegrees=35, twistLimitDegrees=15}
 {
  bone1Name=collar_fl_dyn
  bone2Name=collar_l_dyn
  swing1LimitDegrees=45
  swing1Motion=Free
  swing2LimitDegrees=45
  swing2Motion=Free
  twistLimitDegrees=45
  twistMotion=Free
 }
 {
  bone1Name=collar_l_dyn
  bone2Name=collar_bl_dyn
  swing1LimitDegrees=45
  swing1Motion=Free
  swing2LimitDegrees=45
  swing2Motion=Free
  twistLimitDegrees=45
  twistMotion=Free
 }
 {
  bone1Name=collar_bl_dyn
  bone2Name=collar_b_dyn
  swing1LimitDegrees=45
  swing1Motion=Free
  swing2LimitDegrees=45
  swing2Motion=Free
  twistLimitDegrees=45
  twistMotion=Free
 }
 {
  bone1Name=collar_b_dyn
  bone2Name=collar_br_dyn
  swing1LimitDegrees=45
  swing1Motion=Free
  swing2LimitDegrees=45
  swing2Motion=Free
  twistLimitDegrees=45
  twistMotion=Free
 }
 {
  bone1Name=collar_br_dyn
  bone2Name=collar_r_dyn
  swing1LimitDegrees=45
  swing1Motion=Free
  swing2LimitDegrees=45
  swing2Motion=Free
  twistLimitDegrees=45
  twistMotion=Free
 }
 {
  bone1Name=collar_r_dyn
  bone2Name=collar_fr_dyn
  swing1LimitDegrees=45
  swing1Motion=Free
  swing2LimitDegrees=45
  swing2Motion=Free
  twistLimitDegrees=45
  twistMotion=Free
 }
 {
  bone1Name=foot_r
  bone2Name=calf_r
  swing1LimitDegrees=45
  swing1Motion=Free
  swing2LimitDegrees=45
  swing2Motion=Free
  twistLimitDegrees=45
  twistMotion=Free
 }
 {
  bone1Name=pelvis
  bone2Name=root
  swing1LimitDegrees=45
  swing1Motion=Free
  swing2LimitDegrees=45
  swing2Motion=Free
  twistLimitDegrees=45
  twistMotion=Free
 }
]
