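# Control Rig for the face accessory of the medium male SunBurst Alt character, correcting hair and hat bones as the head moves.
# Three spherical pose readers track head against spine_05: the first weights hair bone offsets, the other two hat_mstr offsets.
# A shared function adds translation, Euler rotation and scale offsets to local transforms.
# hat_back_right_01_mstr gets a constant full-weight nudge, the first reader draws debug, and Is Crouching is unused.
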ControlRigBlueprint /BRCosmetics/Characters/Player/Male/Medium/Bodies/M_MED_SunBurst_Alt/Meshes/Parts/M_MED_Sunburst_Alt_FaceAcc_CtrlRig
graphs [
 ContainedGraph {nodes [
   n1 = RigVMFunctionEntryNode()
   RigUnit_GetTransform = &RigUnit_GetTransform(Space=LocalSpace, bInitial=false, Item=@n1.Joint)
   n2 = &RigVMFunction_MathQuaternionToEuler(RotationOrder=ZYX, Value=@RigUnit_GetTransform.Transform.Rotation)
   > RigUnit_SetTransform = &RigUnit_SetTransform(
    Space=LocalSpace
    Value=(Rotation=(X=0.0132,Y=0.4099,Z=-0.0222,W=0.9118),Translation=(X=23.6045,Y=-6.37,Z=140.1115),Scale3D=(X=1,Y=1,Z=1))
    bInitial=false
    Item=@n1.Joint
    Weight=@n1.Weight
    Value.Translation.Z=&RigVMFunction_MathFloatAdd(B=@n1.Translate Z, A=@RigUnit_GetTransform.Transform.Translation.Z) {template="Add::Execute(in A,in B,out Result)"}.Result
    Value.Scale3D.Z=&RigVMFunction_MathFloatAdd(B=@n1.Scale Z, A=@RigUnit_GetTransform.Transform.Scale3D.Z) {template="Add::Execute(in A,in B,out Result)"}.Result
    Value.Translation.Y=&RigVMFunction_MathFloatAdd(B=@n1.Translate Y, A=@RigUnit_GetTransform.Transform.Translation.Y) {template="Add::Execute(in A,in B,out Result)"}.Result
    Value.Scale3D.Y=&RigVMFunction_MathFloatAdd(B=@n1.Scale Y, A=@RigUnit_GetTransform.Transform.Scale3D.Y) {template="Add::Execute(in A,in B,out Result)"}.Result
    Value.Translation.X=&RigVMFunction_MathFloatAdd(B=@n1.Translate X, A=@RigUnit_GetTransform.Transform.Translation.X) {template="Add::Execute(in A,in B,out Result)"}.Result
    Value.Scale3D.X=&RigVMFunction_MathFloatAdd(B=@n1.Scale X, A=@RigUnit_GetTransform.Transform.Scale3D.X) {template="Add::Execute(in A,in B,out Result)"}.Result
    Value.Rotation=&RigVMFunction_MathQuaternionFromEuler(
     Euler=(X=0,Y=0,Z=0)
     Euler.X=&RigVMFunction_MathFloatAdd(B=@n1.Rotate X, A=@n2.Result.X) {template="Add::Execute(in A,in B,out Result)"}.Result
     Euler.Y=&RigVMFunction_MathFloatAdd(B=@n1.Rotate Y, A=@n2.Result.Y) {template="Add::Execute(in A,in B,out Result)"}.Result
     Euler.Z=&RigVMFunction_MathFloatAdd(B=@n1.Rotate Z, A=@n2.Result.Z) {template="Add::Execute(in A,in B,out Result)"}.Result
    ).Result
   )
   > RigVMFunctionReturnNode()
  ]}
 RigVMFunctionLibrary {nodes [Set Transforms for Bone = RigVMCollapseNode(
    ExecuteContext=()
    Joint=()
    Rotate X=0
    Rotate Y=0
    Rotate Z=0
    Scale X=0
    Scale Y=0
    Scale Z=0
    Translate X=0
    Translate Y=0
    Translate Z=0
    Weight=1
   )]}
 RigVMModel {nodes [
   RigUnit_BeginExecution = &RigUnit_BeginExecution()
   n1 = &RigUnit_SphericalPoseReader(
    ActiveRegionScaleFactors=(PositiveWidth=1,NegativeWidth=1,PositiveHeight=1,NegativeHeight=1)
    ActiveRegionSize=0.2
    Debug=(bDrawDebug=True,DebugScale=25,DebugSegments=20,DebugThickness=0.25)
    DriverAxis=(X=-1,Y=0,Z=0)
    DriverItem=(Type=Bone,Name="head")
    FalloffRegionScaleFactors=(PositiveWidth=1,NegativeWidth=1,PositiveHeight=1,NegativeHeight=1)
    FalloffSize=0.4
    OptionalParentItem=(Type=Bone,Name="spine_05")
    RotationOffset=(X=90,Y=0,Z=40)
    ExecutePin=@RigUnit_BeginExecution.ExecutePin
   )
   > Set Transforms for Bone_3 = RigVMFunctionReferenceNode(
    Joint=(Type=Bone,Name="hair_bl_mstr")
    Rotate X=0
    Rotate Y=-15
    Rotate Z=-15
    Scale X=-0.1
    Scale Y=0
    Scale Z=0
    Translate X=0
    Translate Y=3
    Translate Z=0
    Weight=@n1.OutputParam
   )
   > Set Transforms for Bone_3_1 = RigVMFunctionReferenceNode(
    Joint=(Type=Bone,Name="hair_bl_02_dyn")
    Rotate X=0
    Rotate Y=15
    Rotate Z=0
    Scale X=0
    Scale Y=0
    Scale Z=0
    Translate X=0
    Translate Y=0
    Translate Z=0
    Weight=@n1.OutputParam
   )
   > Set Transforms for Bone_3_1_1 = RigVMFunctionReferenceNode(
    Joint=(Type=Bone,Name="hair_br_mstr")
    Rotate X=0
    Rotate Y=-45
    Rotate Z=0
    Scale X=-0.1
    Scale Y=0
    Scale Z=0
    Translate X=0
    Translate Y=3
    Translate Z=0
    Weight=@n1.OutputParam
   )
   > Set Transforms for Bone_3_1_1_1 = RigVMFunctionReferenceNode(
    Joint=(Type=Bone,Name="hair_br_02_dyn")
    Rotate X=0
    Rotate Y=15
    Rotate Z=0
    Scale X=0
    Scale Y=0
    Scale Z=0
    Translate X=0
    Translate Y=0
    Translate Z=0
    Weight=@n1.OutputParam
   )
   > Set Transforms for Bone_3_1_1_1_1 = RigVMFunctionReferenceNode(
    Joint=(Type=Bone,Name="hair_mid_bl_mstr")
    Rotate X=0
    Rotate Y=-25
    Rotate Z=0
    Scale X=0
    Scale Y=0
    Scale Z=0
    Translate X=0
    Translate Y=0
    Translate Z=3
    Weight=@n1.OutputParam
   )
   > Set Transforms for Bone_3_1_1_1_2 = RigVMFunctionReferenceNode(
    Joint=(Type=Bone,Name="hair_mid_br_mstr")
    Rotate X=0
    Rotate Y=-15
    Rotate Z=0
    Scale X=0
    Scale Y=0
    Scale Z=0
    Translate X=0
    Translate Y=7
    Translate Z=0
    Weight=@n1.OutputParam
   )
   > Set Transforms for Bone_3_1_1_1_1_1 = RigVMFunctionReferenceNode(
    Joint=(Type=Bone,Name="hair_mid_fl_mstr")
    Rotate X=0
    Rotate Y=-15
    Rotate Z=0
    Scale X=0
    Scale Y=0
    Scale Z=0
    Translate X=0
    Translate Y=0
    Translate Z=0
    Weight=@n1.OutputParam
   )
   > Set Transforms for Bone_3_1_1_1_2_1 = RigVMFunctionReferenceNode(
    Joint=(Type=Bone,Name="hair_mid_fr_mstr")
    Rotate X=0
    Rotate Y=-5
    Rotate Z=0
    Scale X=0
    Scale Y=0
    Scale Z=0
    Translate X=0
    Translate Y=0
    Translate Z=0
    Weight=@n1.OutputParam
   )
   > n2 = &RigUnit_SphericalPoseReader(
    ActiveRegionScaleFactors=(PositiveWidth=1,NegativeWidth=1,PositiveHeight=1,NegativeHeight=1)
    ActiveRegionSize=0.2
    Debug=(bDrawDebug=false,DebugScale=25,DebugSegments=20,DebugThickness=0.25)
    DriverAxis=(X=-1,Y=0,Z=0)
    DriverItem=(Type=Bone,Name="head")
    FalloffRegionScaleFactors=(PositiveWidth=0.8,NegativeWidth=0.8,PositiveHeight=1,NegativeHeight=1)
    FalloffSize=0.4
    OptionalParentItem=(Type=Bone,Name="spine_05")
    RotationOffset=(X=160,Y=-15,Z=0)
   )
   > Set Transforms for Bone_3_1_1_1_2_1_1_3 = RigVMFunctionReferenceNode(
    Joint=(Type=Bone,Name="hat_mstr")
    Rotate X=0
    Rotate Y=10
    Rotate Z=0
    Scale X=0
    Scale Y=0
    Scale Z=0
    Translate X=1
    Translate Y=-1.58
    Translate Z=0
    Weight=@n2.OutputParam
   )
   > Set Transforms for Bone_3_1_1_1_2_1_1_3_1 = RigVMFunctionReferenceNode(
    Joint=(Type=Bone,Name="hat_back_right_01_mstr")
    Rotate X=0
    Rotate Y=0
    Rotate Z=0
    Scale X=0
    Scale Y=0
    Scale Z=0
    Translate X=0
    Translate Y=1
    Translate Z=0
    Weight=1
   )
   > n3 = &RigUnit_SphericalPoseReader(
    ActiveRegionScaleFactors=(PositiveWidth=0.5,NegativeWidth=0.5,PositiveHeight=0.5,NegativeHeight=0.5)
    ActiveRegionSize=0.2
    Debug=(bDrawDebug=false,DebugScale=25,DebugSegments=20,DebugThickness=0.25)
    DriverAxis=(X=-1,Y=0,Z=0)
    DriverItem=(Type=Bone,Name="head")
    FalloffRegionScaleFactors=(PositiveWidth=0.4,NegativeWidth=0.4,PositiveHeight=0.5,NegativeHeight=0.5)
    FalloffSize=0.4
    OptionalParentItem=(Type=Bone,Name="spine_05")
    RotationOffset=(X=220,Y=-105,Z=0)
   )
   > Set Transforms for Bone_3_1_1_1_2_1_1_3_2 = RigVMFunctionReferenceNode(
    Joint=(Type=Bone,Name="hat_mstr")
    Rotate X=0
    Rotate Y=0
    Rotate Z=0
    Scale X=0
    Scale Y=0
    Scale Z=0
    Translate X=0.25
    Translate Y=0.2
    Translate Z=0
    Weight=@n3.OutputParam
   )
  ]}
]
hierarchy [
 root {loc=<0 0 0>, rot=<0 0 0 1>}
 pelvis {parent=root, loc=<0 -0.7071 93.1191>, rot=<0.0127 0.707 -0.0127 -0.707>}
 spine_01 {parent=pelvis, loc=<6.7494 -0.6388 0>, rot=<0 0 0.034 -0.9994>}
 spine_02 {parent=spine_01, loc=<7.2518 -0.035 0>, rot=<0 0 0.0238 -0.9997>}
 spine_03 {parent=spine_02, loc=<6.0104 0 0>, rot=<0 0 0.0149 -0.9999>}
 spine_04 {parent=spine_03, loc=<8.1024 0 0>, rot=<0 0 -0.0132 -0.9999>}
 spine_05 {parent=spine_04, loc=<10.0146 0 0>, rot=<0 0 -0.0015 -1>}
 neck_01 {parent=spine_05, loc=<11.3138 0.1529 0>, rot=<0 0 -0.2003 -0.9797>}
 neck_02 {parent=neck_01, loc=<4.5627 0 0>, rot=<0 0 0.0504 -0.9987>}
 head {parent=neck_02, loc=<5.3914 0 0>, rot=<0 0 0.115 -0.9934>}
 bang_r_mstr {parent=head, loc=<11.5064 7.4904 5.2432>, rot=<0.2124 -0.0141 0.9744 -0.073>}
 bang_r_01_dyn {parent=bang_r_mstr, loc=<0 0 0>, rot=<0 0 0 1>}
 bang_r_02_dyn {parent=bang_r_01_dyn, loc=<6.6028 0 0>, rot=<-0.0105 -0.09 0.0251 -0.9956>}
 bang_r_03_dyn {parent=bang_r_02_dyn, loc=<5.5587 0 0>, rot=<-0.0056 -0.1233 0.0451 -0.9913>}
 hair_mid_fr_mstr {parent=head, loc=<9.6713 3.0428 6.8536>, rot=<0.1827 -0.0059 0.9823 -0.04>}
 hair_mid_fr_01_dyn {parent=hair_mid_fr_mstr, loc=<0 0 0>, rot=<0 0 0 1>}
 hair_mid_fr_02_dyn {parent=hair_mid_fr_01_dyn, loc=<4.8372 0 0>, rot=<-0.0066 -0.099 0.0066 -0.995>}
 hair_mid_fr_03_dyn {parent=hair_mid_fr_02_dyn, loc=<6.4225 0 0>, rot=<-0.0019 -0.0627 0.0299 -0.9976>}
 hair_mid_br_mstr {parent=head, loc=<10.2499 -2.5488 6.7605>, rot=<0.1842 -0.0081 0.9815 -0.0513>}
 hair_mid_br_01_dyn {parent=hair_mid_br_mstr, loc=<0 0 0>, rot=<0 0 0 1>}
 hair_mid_br_02_dyn {parent=hair_mid_br_01_dyn, loc=<5.4263 0 0>, rot=<-0.0106 -0.1496 0.025 -0.9884>}
 hair_mid_br_03_dyn {parent=hair_mid_br_02_dyn, loc=<6.3418 0 0>, rot=<-0.0008 -0.0349 0.0228 -0.9991>}
 hair_br_mstr {parent=head, loc=<10.2912 -6.2947 2.1284>, rot=<0.1439 -0.6918 0.6923 -0.1462>}
 hair_br_01_dyn {parent=hair_br_mstr, loc=<0 0 0>, rot=<0 0 0 1>}
 hair_br_02_dyn {parent=hair_br_01_dyn, loc=<5.4885 0 0>, rot=<0.0033 -0.1712 0.0139 -0.9851>}
 hair_br_03_dyn {parent=hair_br_02_dyn, loc=<6.3335 0 0>, rot=<-0.0004 -0.0304 -0.0127 -0.9995>}
 hair_bl_mstr {parent=head, loc=<10.3309 -6.2773 -3.3795>, rot=<0.1276 -0.6884 0.6953 -0.1621>}
 hair_bl_01_dyn {parent=hair_bl_mstr, loc=<0 0 0>, rot=<0 0 0 1>}
 hair_bl_02_dyn {parent=hair_bl_01_dyn, loc=<5.5377 0 0>, rot=<0.0001 -0.1711 0.0004 -0.9853>}
 hair_bl_03_dyn {parent=hair_bl_02_dyn, loc=<6.3391 0 0>, rot=<0.0013 -0.006 0.0245 -0.9997>}
 hair_mid_bl_mstr {parent=head, loc=<10.1744 -1.1318 -6.0539>, rot=<0.0445 -0.9805 0.0071 -0.1912>}
 hair_mid_bl_01_dyn {parent=hair_mid_bl_mstr, loc=<0 0 0>, rot=<0 0 0 1>}
 hair_mid_bl_02_dyn {parent=hair_mid_bl_01_dyn, loc=<5.3814 0 0>, rot=<0.0104 -0.1454 -0.0108 -0.9893>}
 hair_mid_bl_03_dyn {parent=hair_mid_bl_02_dyn, loc=<6.3586 0 0>, rot=<0.0014 -0.0462 -0.0302 -0.9985>}
 hair_mid_fl_mstr {parent=head, loc=<9.7136 5.1262 -6.2197>, rot=<-0.0067 -0.9815 -0.0029 -0.1913>}
 hair_mid_fl_01_dyn {parent=hair_mid_fl_mstr, loc=<0 0 0>, rot=<0 0 0 1>}
 hair_mid_fl_02_dyn {parent=hair_mid_fl_01_dyn, loc=<4.9228 0 0>, rot=<0.0032 -0.1409 0.044 -0.989>}
 hair_mid_fl_03_dyn {parent=hair_mid_fl_02_dyn, loc=<5.7533 0 0>, rot=<0.0028 -0.0511 -0.0122 -0.9986>}
 hair_bang_l_mstr {parent=head, loc=<11.9192 8.8499 -3.0856>, rot=<-0.0246 -0.9179 -0.2764 -0.2838>}
 hair_bang_l_01_dyn {parent=hair_bang_l_mstr, loc=<0 0 0>, rot=<0 0 0 1>}
 hair_bang_l_02_dyn {parent=hair_bang_l_01_dyn, loc=<2.7147 0 0>, rot=<0.1636 -0.2193 -0.2566 -0.927>}
 hair_bang_l_03_dyn {parent=hair_bang_l_02_dyn, loc=<4.4874 0 0>, rot=<-0.2397 -0.0083 0.1668 -0.9564>}
 hair_bang_l_04_dyn {parent=hair_bang_l_02_dyn, loc=<3.4646 -3.7727 0.2687>, rot=<-0.2556 -0.0965 0.1038 -0.9563>}
 hat_mstr {parent=head, loc=<0 0 0>, rot=<0.4733 -0.5254 0.5252 -0.4734>}
 hat_back_01_mstr {parent=hat_mstr, loc=<8.5509 -0.005 10.0883>, rot=<0 0 0 1>}
 hat_back_left_01_mstr {parent=hat_back_01_mstr, loc=<-1.4952 4.6164 -0.7287>, rot=<0.016 0.052 -0.2761 -0.9596>}
 hat_back_right_01_mstr {parent=hat_back_01_mstr, loc=<-1.3479 -5.6831 -0.2226>, rot=<0.0081 0.143 0.2303 -0.9625>}
]
variables [Is Crouching {type=bool, default=False}]
